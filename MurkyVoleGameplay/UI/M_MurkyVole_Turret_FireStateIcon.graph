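# UI material for the MurkyVole turret fire-state icon, using alpha-composite blending.
# A single grayscale fire-state texture drives two branches through its red channel.
# One branch adds the mask to a constant vector left at its default.
# The other uses the mask to interpolate up to a scalar parameter, spelled "Tramslucency", which scales the mask for adjustable translucency.
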Material /MurkyVoleGameplay/UI/M_MurkyVole_Turret_FireStateIcon
graph {
 properties {bAutomaticallySetUsageInEditor=false, blendMode=BLEND_AlphaComposite, materialDomain=MD_UI, usedShadingModels=Default Lit}
 nodes [
  n1 = TextureSample() {samplerType=SAMPLERTYPE_Grayscale, texture=&/T_MurkyVole_Turret_FireState}
  Add(A=Constant3Vector(), B=@n1.R)
  LinearInterpolate(B=ScalarParameter() {defaultValue=0.25, parameterName=Tramslucency}, Alpha=@n1.R)
 ]
}
