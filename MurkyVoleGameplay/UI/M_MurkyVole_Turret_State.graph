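# UI material for a MurkyVole turret state display, drawing a rounded box frame with alpha-composite blending.
# MF_UI_Box calls build an outline and a soft inner edge from an inverted glow-range fill, both masked by the outer box fill and summed.
# That mask, added to a constant color, gives a grayscale shade and also drives a lerp up to the scalar parameter "Tramslucency" for translucency.
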Material /MurkyVoleGameplay/UI/M_MurkyVole_Turret_State
graph {
 properties {bAutomaticallySetUsageInEditor=false, blendMode=BLEND_AlphaComposite, materialDomain=MD_UI, usedShadingModels=Default Lit}
 nodes [
  n1 = Constant() {r=0.98}
  n2 = Constant() {r=0.5}
  n3 = Constant() {r=0.175}
  n4 = StaticBool() {value=true}
  n5 = MaterialFunctionCall(Size=@n1, CornerRoundness=Constant() {r=0.5}, KeepAspectRatio=@n4) {materialFunction=&MF_UI_Box}
  n6 = OneMinus(None=MaterialFunctionCall(Size=Constant() {r=0.9}, CornerRoundness=Constant() {r=0.45}, GlowMin=Constant() {r=-1.2}, GlowMax=Constant() {r=0.4}, KeepAspectRatio=@n4) {materialFunction=&MF_UI_Box}.Fill)
  n7 = Add(
   A=Multiply(
    A=MaterialFunctionCall(Size=@n1, CornerRoundness=@n2, OutlineThickness=@n3, KeepAspectRatio=@n4) {materialFunction=&MF_UI_Box}.Outline
    B=@n5.Fill
   )
   B=Multiply(A=@n5.Fill, B=Add(A=@n6) {constB=-0.7})
  )
  Add(A=Constant3Vector(), B=@n7)
  LinearInterpolate(B=ScalarParameter() {defaultValue=0.25, parameterName=Tramslucency}, Alpha=@n7)
 ]
}
