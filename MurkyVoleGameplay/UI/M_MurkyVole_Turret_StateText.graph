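# M_MurkyVole_Turret_StateText is an additive UI-domain material, named for turret state text in MurkyVole.
# Its graph holds two separate fragments: an Add fed by a default constant vector, and a Lerp taking a "Tramslucency" (misspelled) scalar parameter as B with a constant white alpha.
# No connection from either fragment to a material output is shown, so the graph appears unfinished.
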Material /MurkyVoleGameplay/UI/M_MurkyVole_Turret_StateText
graph {
 properties {bAutomaticallySetUsageInEditor=false, blendMode=BLEND_Additive, materialDomain=MD_UI, usedShadingModels=Default Lit}
 nodes [
  Add(A=Constant3Vector())
  LinearInterpolate(B=ScalarParameter() {defaultValue=0.4, parameterName=Tramslucency}, Alpha=Constant3Vector() {constant=<rgba 1 1 1 0>})
 ]
}
