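# Reticle extension widget for the MurkyVole turret weapon that shows whether auto targeting is on and an aiming status line.
# On construct it caches the turret and owning pawn, then uses async gameplay tag waits on the pawn's ManualMode tag.
# ToggleTurretMode treats tag removal as active, switching label texts and recoloring the state icon, backplate and label.
# SetTouchPosition shifts the root grid for mobile touch players.
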WidgetBlueprint /MurkyVoleGameplay/Gameplay/Turret/UI/WBP_Reticle_MurkyVole_Turret
graphs [
 EventGraph {nodes [
   n1 = ToggleTurretMode!() -> (OutputDelegate, then, Active?:bool=~) {is=title K2Node_CustomEvent}
   n2 = Select(
    Option 0:Linear Color Structure=Get InactiveColor() -> (InactiveColor:Linear Color Structure)
    Option 1:Linear Color Structure=Get ActiveColor() -> (ActiveColor:Linear Color Structure)
    Index=@n1.Active?
   ) -> (ReturnValue:Linear Color Structure)
   > SetColorAndOpacity(
    self:Image Object Reference=[Get AutoFireIcon() -> (AutoFireIcon:Image Object Reference), Get StateBackplate() -> (StateBackplate:Image Object Reference)]
    InColorAndOpacity:Linear Color Structure=@n2
   )
   n3 = Get StateLabel() -> (StateLabel:Common Text Object Reference)
   > Set Text(Text=SelectText(A="LOCTEXT(\"On\")", B="LOCTEXT(\"Off\")", bPickA=@n1.Active?), self=@n3)
   > SetColorAndOpacity(self:Text Object Reference=@n3, InColorAndOpacity:Slate Color Structure=Make Slate Color(SpecifiedColor=@n2))
   > Set Text(
    Text=SelectText(A="LOCTEXT(\"Searching\")", B="LOCTEXT(\"- - -\")", bPickA=@n1.Active?)
    self=Get AimingStatus() -> (AimingStatus:Common Text Object Reference)
   )
   Event Construct()
   > SetTouchPosition()
   > n4 = Cast To B_Athena_MurkyVole_Turret!(
    execute
    Object=GetEquippedWeapon!(self:Fort Weapon Reticle Extension Widget Base Object Reference) -> (ReturnValue:Fort Weapon Object Reference)
   ) -> (then, CastFailed, AsB Athena Murky Vole Turret:B Athena Murky Vole Turret Object Reference) {is=title K2Node_DynamicCast}
   > Set Turret(Turret:B Athena Murky Vole Turret Object Reference=@n4) -> (Output_Get:B Athena Murky Vole Turret Object Reference)
   > n5 = Cast To FortPlayerPawn(Object=GetOwningPlayerPawn!(self:User Widget Object Reference) -> (ReturnValue:Pawn Object Reference))
   > n6 = Set Pawn(Pawn:Fort Player Pawn Object Reference=@n5) -> (Output_Get:Fort Player Pawn Object Reference)
   > n7 = WaitGameplayTagAddToActor!(
    execute:Exec
    TargetActor:Actor Object Reference=@n6
    Tag:Gameplay Tag Structure=(TagName="Granted.Athena.MurkyVole.Turret.ManualMode")
    OnlyTriggerOnce:bool=false
   ) -> (then:Exec, AsyncTaskProxy:Ability Async Wait Gameplay Tag Added Object Reference, Added:Exec) {is=title K2Node_AsyncAction}
   > n8 = WaitGameplayTagRemoveFromActor!(
    execute:Exec
    TargetActor:Actor Object Reference=@n6
    Tag:Gameplay Tag Structure=(TagName="Granted.Athena.MurkyVole.Turret.ManualMode")
    OnlyTriggerOnce:bool=false
   ) -> (then:Exec, AsyncTaskProxy:Ability Async Wait Gameplay Tag Removed Object Reference, Removed:Exec) {is=title K2Node_AsyncAction}
   ToggleTurretMode!(execute=@n8.Removed, self, Active?:bool=true) -> (then) {is=title K2Node_CallFunction}
   ToggleTurretMode(execute=@n7.Added, Active?=false)
   Event Tick()
  ]}
 SetTouchPosition {nodes [
   SetTouchPosition() {is=title K2Node_FunctionEntry}
   n1 = GetOwningPlayer()
   > Branch(Condition=AND Boolean(A=IsUsingTouch(PlayerController=@n1), B=IsMobileGame()))
   > SetRenderTranslation(self=Get GridPanel_14() -> (GridPanel_14:Grid Panel Object Reference), Translation=(X=180,Y=50))
  ]}
]
variables [
 ActiveColor {type=LinearColor, default=<rgba 0.0685 1 0.1195 1>}
 InactiveColor {type=LinearColor, default=<rgba 1 1 1 0.534>}
 Pawn {type=FortPlayerPawn}
 Turret {type=B_Athena_MurkyVole_Turret_C, default=null}
]
widgets {
 info {
  inheritedWidgetCount=1
  parentClass=/Script/FortniteUI.FortWeaponReticleExtensionWidgetBase
  rootWidgetClass=&GridPanel
  widgetCount=13
 }
 widgets [
  GridPanel_14 {
   widgetClassPath=&GridPanel
   bIsVariable=true
   widget=GridPanel_14
   properties {renderTransform {angle=0, scale=<1 1>, shear=<0 0>, translation=<180 100>}, rowFill [1,1]}
  }
  HorizontalBox_172 {
   widgetClassPath=&HorizontalBox
   parent=GridPanel_14
   widget=HorizontalBox_172
   slot {class=GridSlot, properties {horizontalAlignment=HAlign_Left, verticalAlignment=VAlign_Top}}
  }
  Image_129 {
   widgetClassPath=&Image
   bIsVariable=true
   parent=HorizontalBox_172
   widget=Image_129
   properties {brush {
     bIsDynamicallyLoaded=false
     drawAs=Image
     imageType=FullColor
     mirroring=NoMirror
     resourceName=None
     resourceObject=/MurkyVoleGameplay/UI/T_MurkyVole_Turret_BracketB
     tiling=NoTile
     imageSize=<9.4566 32.5944>
     margin=<ltrb 0 0 0 0>
     outlineSettings {
      bUseBrushTransparency=false
      roundingType=HalfHeightRadius
      width=0
      color {colorUseRule=UseColor_Specified, specifiedColor=<rgba 0 0 0 0>}
      cornerRadii=<0 0 0 0>
     }
     tintColor {colorUseRule=UseColor_Specified, specifiedColor=<rgba 1 1 1 0.2>}
     uVRegion {bIsValid=false, max=<0 0>, min=<0 0>}
    }}
   slot {class=HorizontalBoxSlot, properties {horizontalAlignment=HAlign_Left, verticalAlignment=VAlign_Top, padding=<ltrb -8 0 0 -5>}}
  }
  Image {
   widgetClassPath=&Image
   bIsVariable=true
   parent=HorizontalBox_172
   widget=Image
   properties {
    brush {
     bIsDynamicallyLoaded=false
     drawAs=Image
     imageType=FullColor
     mirroring=NoMirror
     resourceName=None
     resourceObject=/MurkyVoleGameplay/UI/T_MurkyVole_Turret_AimState
     tiling=NoTile
     imageSize=<18 18>
     margin=<ltrb 0 0 0 0>
     outlineSettings {
      bUseBrushTransparency=false
      roundingType=HalfHeightRadius
      width=0
      color {colorUseRule=UseColor_Specified, specifiedColor=<rgba 0 0 0 0>}
      cornerRadii=<0 0 0 0>
     }
     tintColor {colorUseRule=UseColor_Specified, specifiedColor=<rgba 1 1 1 1>}
     uVRegion {bIsValid=false, max=<0 0>, min=<0 0>}
    }
    colorAndOpacity=<rgba 1 1 1 0.6>
   }
   slot {class=HorizontalBoxSlot, properties {horizontalAlignment=HAlign_Left, verticalAlignment=VAlign_Center, padding=<ltrb -2.5 0 0 0>}}
  }
  AimingStatus {
   widgetClassPath=&CommonTextBlock
   bIsVariable=true
   parent=HorizontalBox_172
   widget=AimingStatus
   properties {
    text=searching
    textTransformPolicy=ToUpper
    colorAndOpacity {colorUseRule=UseColor_Specified, specifiedColor=<rgba 1 1 1 0.75>}
    font {
     bForceMonospaced=false
     bMaterialIsStencil=false
     fontMaterial=null
     fontObject=&BurbankSmall
     letterSpacing=0
     monospacedWidth=1
     size=9
     skewAmount=0
     typefaceFontName=Medium
     outlineSettings {
      bApplyOutlineToDropShadows=false
      bMiteredCorners=false
      bSeparateFillAlpha=false
      outlineMaterial=null
      outlineSize=0
      outlineColor=<rgba 0 0 0 1>
     }
    }
    renderTransform {angle=0, scale=<0.95 1>, shear=<0 0>, translation=<0 0>}
    renderTransformPivot=<0 0.5>
    shadowOffset=<0 0>
   }
   slot {class=HorizontalBoxSlot, properties {horizontalAlignment=HAlign_Left, verticalAlignment=VAlign_Center, padding=<ltrb 0 0.5 0 0>}}
  }
  HorizontalBox_64 {
   widgetClassPath=&HorizontalBox
   parent=GridPanel_14
   widget=HorizontalBox_64
   slot {class=GridSlot, properties {horizontalAlignment=HAlign_Left, row=1, verticalAlignment=VAlign_Top, padding=<ltrb 2 0 0 -10>}}
  }
  Image_51 {
   widgetClassPath=&Image
   bIsVariable=true
   parent=HorizontalBox_64
   widget=Image_51
   properties {brush {
     bIsDynamicallyLoaded=false
     drawAs=Image
     imageType=FullColor
     mirroring=NoMirror
     resourceName=None
     resourceObject=/MurkyVoleGameplay/UI/T_MurkyVole_Turret_BracketA
     tiling=NoTile
     imageSize=<10 35.7426>
     margin=<ltrb 0 0 0 0>
     outlineSettings {
      bUseBrushTransparency=false
      roundingType=HalfHeightRadius
      width=0
      color {colorUseRule=UseColor_Specified, specifiedColor=<rgba 0 0 0 0>}
      cornerRadii=<0 0 0 0>
     }
     tintColor {colorUseRule=UseColor_Specified, specifiedColor=<rgba 1 1 1 0.25>}
     uVRegion {bIsValid=false, max=<0 0>, min=<0 0>}
    }}
   slot {class=HorizontalBoxSlot, properties {horizontalAlignment=HAlign_Left, verticalAlignment=VAlign_Top}}
  }
  AutoFireIcon {
   widgetClassPath=&Image
   bIsVariable=true
   parent=HorizontalBox_64
   widget=AutoFireIcon
   properties {brush {
     bIsDynamicallyLoaded=false
     drawAs=Image
     imageType=FullColor
     mirroring=NoMirror
     resourceName=None
     resourceObject=/MurkyVoleGameplay/UI/M_MurkyVole_Turret_FireStateIcon
     tiling=NoTile
     imageSize=<18 18>
     margin=<ltrb 0 0 0 0>
     outlineSettings {
      bUseBrushTransparency=false
      roundingType=HalfHeightRadius
      width=0
      color {colorUseRule=UseColor_Specified, specifiedColor=<rgba 0 0 0 0>}
      cornerRadii=<0 0 0 0>
     }
     tintColor {colorUseRule=UseColor_Specified, specifiedColor=<rgba 1 1 1 1>}
     uVRegion {bIsValid=false, max=<0 0>, min=<0 0>}
    }}
   slot {class=HorizontalBoxSlot, properties {horizontalAlignment=HAlign_Left, verticalAlignment=VAlign_Center}}
  }
  CommonTextBlock_61 {
   widgetClassPath=&CommonTextBlock
   parent=HorizontalBox_64
   widget=CommonTextBlock_61
   properties {
    text=AUTO TARGETING
    textTransformPolicy=ToUpper
    font {
     bForceMonospaced=false
     bMaterialIsStencil=false
     fontMaterial=null
     fontObject=&BurbankSmall
     letterSpacing=0
     monospacedWidth=1
     size=13.5
     skewAmount=0
     typefaceFontName=Bold
     outlineSettings {
      bApplyOutlineToDropShadows=false
      bMiteredCorners=false
      bSeparateFillAlpha=false
      outlineMaterial=null
      outlineSize=0
      outlineColor=<rgba 0 0 0 0.5>
     }
    }
    renderTransformPivot=<0 0.5>
    shadowOffset=<0 0>
   }
   slot {class=HorizontalBoxSlot, properties {horizontalAlignment=HAlign_Left, verticalAlignment=VAlign_Center, padding=<ltrb 4 0 0 0>}}
  }
  Overlay_71 {
   widgetClassPath=&Overlay
   parent=HorizontalBox_64
   widget=Overlay_71
   slot {class=HorizontalBoxSlot, properties {horizontalAlignment=HAlign_Left, verticalAlignment=VAlign_Center, padding=<ltrb 7 0 0 0>}}
  }
  StateBackplate {
   widgetClassPath=&Image
   bIsVariable=true
   parent=Overlay_71
   widget=StateBackplate
   properties {
    brush {
     bIsDynamicallyLoaded=false
     drawAs=Image
     imageType=FullColor
     mirroring=NoMirror
     resourceName=None
     resourceObject=/MurkyVoleGameplay/UI/M_MurkyVole_Turret_State
     tiling=NoTile
     imageSize=<76 28>
     margin=<ltrb 0 0 0 0>
     outlineSettings {
      bUseBrushTransparency=false
      roundingType=HalfHeightRadius
      width=0
      color {colorUseRule=UseColor_Specified, specifiedColor=<rgba 0 0 0 0>}
      cornerRadii=<0 0 0 0>
     }
     tintColor {colorUseRule=UseColor_Specified, specifiedColor=<rgba 1 1 1 1>}
     uVRegion {bIsValid=false, max=<0 0>, min=<0 0>}
    }
    colorAndOpacity=<rgba 1 1 1 0.304>
   }
   slot {class=OverlaySlot, properties {horizontalAlignment=HAlign_Center, verticalAlignment=VAlign_Center}}
  }
  StateLabel {
   widgetClassPath=&CommonTextBlock
   bIsVariable=true
   parent=Overlay_71
   widget=StateLabel
   properties {
    text=ON
    textTransformPolicy=ToUpper
    font {
     bForceMonospaced=false
     bMaterialIsStencil=false
     fontMaterial=/MurkyVoleGameplay/UI/M_MurkyVole_Turret_StateText
     fontObject=&BurbankSmall
     letterSpacing=0
     monospacedWidth=1
     size=13.5
     skewAmount=0
     typefaceFontName=Bold
     outlineSettings {
      bApplyOutlineToDropShadows=false
      bMiteredCorners=false
      bSeparateFillAlpha=false
      outlineMaterial=null
      outlineSize=0
      outlineColor=<rgba 0 0 0 0.5>
     }
    }
    margin=<ltrb 0 -5 0 -5>
    shadowOffset=<0 0>
   }
   slot {class=OverlaySlot, properties {horizontalAlignment=HAlign_Center, verticalAlignment=VAlign_Center}}
  }
  TutorialHighlight {widgetClassPath=&TutorialHighlight}
 ]
}
